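# Unlit translucent Niagara sprite material for the PennantSeas backpack cosmetic, apparently a dark background effect, judging by its name.
# Particle color RGB runs through the shared FX emissive adjustment function.
# Particle alpha is multiplied by a view-angle fade and by the alpha of a pennant diffuse texture, exposed as a mask parameter, forming a shaped mask that fades with viewing angle.
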Material /BRCosmetics/Accessories/FORT_Backpacks/Backpack_M_MED_PennantSeas/Material/M_PennantSeas_DarkBG
graph {
 properties {bUsedWithNiagaraSprites=true, blendMode=BLEND_Translucent, pixelDepthOffsetMode=PDOM_Legacy, shadingModel=MSM_Unlit, usedShadingModels=Unlit}
 nodes [
  n1 = ParticleColor()
  MaterialFunctionCall(Emissive=@n1.RGB) {materialFunction=&MF_FX_EmissiveAdjustments}
  n2 = Constant() {r=1}
  n3 = TextureSampleParameter2D() {
   parameterName=Mask_from_Diffuse
   texture=/BRCosmetics/Accessories/FORT_Backpacks/Backpack_M_MED_PennantSeas/Texture/B_18/T_PennantSeasonsTide_18_D
  }
  Multiply(
   A=Multiply(
    A=@n1.A
    B=MaterialFunctionCall("FadeSensitivity (optional)"=@n2) {materialFunction=&FadeBasedOnViewAngle}.Result
   )
   B=@n3.A
  )
 ]
}
